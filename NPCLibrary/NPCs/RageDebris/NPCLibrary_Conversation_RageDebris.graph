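# Conversation database for the RageDebris NPC, who calls himself Skillet, providing his greeting and top-level conversation hub.
# A single entry point leads to a speech node whose contextual lines cover first meetings, per-match conversation counts and reactions to specific player outfits, with priority curves for introductions and outfit reactions.
# That node fans out through link nodes to separately tagged entries for quests, basic and deluxe services, dialogue, explosive expertise, and seasonal story content.
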ConversationDatabase /NPCLibrary/NPCs/RageDebris/NPCLibrary_Conversation_RageDebris
entry_points [{tag=Conversation.Entry.RageDebris, destinations [#1]}]
nodes [
 ConversationEntryPointNode_7 {
  id=#1
  class=ConversationEntryPointNode
  outputs [#2]
  properties {entryTag {tagName=Conversation.Entry.RageDebris}}
 }
 Speech_C_4 {
  id=#2
  class=Speech_C
  outputs [
   #3
   #4
   #5
   #6
   #7
   #8
   #9
   #10
   #11
  ]
  properties {
   nodeName=""
   generalConfig {
    defaultMessage="Beautiful day to set off some booms, huh?"
    contextualMessages [
     {
      message="Heyo there, ratso. Ye can call me Skillet!"
      contextRequirements [{requirement {class=FortControllerRequirement_IsFirstEverConversationWithNPC}}]
      priority {curve {rowName=Tandem.Speech.Priority.FirstEverInteraction}}
     }
     {
      message="Beautiful day to set off some booms, huh?"
      contextRequirements [{requirement {class=FortControllerRequirement_CheckNPCMatchConversationCount, properties {nPCUniqueIdTag {tagName=AISpawnerData.Type.Tandem.RageDebris}}}}]
      priority {curve {rowName=None}}
     }
     {
      message="I got a short fuse, sure, but I'm worth the heat!"
      contextRequirements [{requirement {class=FortControllerRequirement_CheckNPCMatchConversationCount, properties {nPCUniqueIdTag {tagName=AISpawnerData.Type.Tandem.RageDebris}}}}]
      priority {curve {rowName=None}}
     }
     {
      message="I retired from makin' explosives to spend more time setting 'em off!"
      contextRequirements [{requirement {class=FortControllerRequirement_CheckNPCMatchConversationCount, properties {nPCUniqueIdTag {tagName=AISpawnerData.Type.Tandem.RageDebris}}}}]
      priority {curve {rowName=None}}
     }
     {
      message="Hey there, gold man. Got my ear to the ground for ye."
      contextRequirements [{requirement {class=FortControllerRequirement_HasCID, properties {allowedCIds [&Character_AuraCop_Bullet]}}}]
      priority {curve {rowName=Tandem.Speech.Priority.OutfitReaction}}
     }
     {
      message="Ye know the gold man ain't exactly trustworthy, right?"
      contextRequirements [{requirement {class=FortControllerRequirement_HasCID, properties {allowedCIds [/BRCosmetics/Athena/Items/Cosmetics/Characters/Character_CasinoReaper_Die]}}}]
      priority {curve {rowName=Tandem.Speech.Priority.OutfitReaction}}
     }
     {
      message=I ain't being sneaky. Just mindin' my own business...
      contextRequirements [{requirement {class=FortControllerRequirement_HasCID, properties {allowedCIds [&Character_IronLilac_Sly]}}}]
      priority {curve {rowName=Tandem.Speech.Priority.OutfitReaction}}
     }
    ]
   }
  }
 }
 ConversationLinkNode_8 {
  id=#11
  class=ConversationLinkNode
  properties {remoteEntryTag {tagName=Conversation.Entry.Season36.RageDebris.CharacterStories}}
 }
 ConversationLinkNode_7 {id=#10, class=ConversationLinkNode, properties {remoteEntryTag {tagName=Conversation.Entry.DeluxeServices}}}
 ConversationLinkNode_6 {
  id=#9
  class=ConversationLinkNode
  properties {remoteEntryTag {tagName=Conversation.Entry.Expertise.Explosive}}
 }
 ConversationLinkNode_4 {
  id=#8
  class=ConversationLinkNode
  properties {remoteEntryTag {tagName=Conversation.Entry.Season34.Expertise.Explosive.Introduction}}
 }
 ConversationLinkNode_5 {
  id=#7
  class=ConversationLinkNode
  properties {remoteEntryTag {tagName=Conversation.Entry.Season34.Expertise.Explosive.Finish}}
 }
 ConversationLinkNode_3 {
  id=#6
  class=ConversationLinkNode
  properties {remoteEntryTag {tagName=Conversation.Entry.RageDebris.Dialogue}}
 }
 ConversationLinkNode_2 {
  id=#5
  class=ConversationLinkNode
  properties {remoteEntryTag {tagName=Conversation.Entry.Season34.Story.P01.Q02}}
 }
 ConversationLinkNode_0 {id=#4, class=ConversationLinkNode, properties {remoteEntryTag {tagName=Conversation.Entry.BasicServices}}}
 ConversationLinkNode_1 {id=#3, class=ConversationLinkNode, properties {remoteEntryTag {tagName=Conversation.Entry.NPCLibrary.Quests}}}
]
