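# Conversation database for the "Ask about..." dialogue of NPC RageDebris, who calls himself Skillet.
# The entry speech branches to choices about his name and Midas, a Back option, and a link to the separate BaronEsc entry tag.
# Both topics play as linear chains of speech nodes that converge on one shared ReturnToStart node shown as BACK.
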ConversationDatabase /NPCLibrary/NPCs/RageDebris/NPCLibrary_Conversation_RageDebris_Dialogue
entry_points [{tag=Conversation.Entry.RageDebris.Dialogue, destinations [#1]}]
nodes [
 ConversationEntryPointNode_0 {
  id=#1
  class=ConversationEntryPointNode
  outputs [#2]
  properties {entryTag {tagName=Conversation.Entry.RageDebris.Dialogue}}
 }
 Speech_C_10 {
  id=#2
  class=Speech_C
  outputs [#3, #4, #5, #6]
  properties {nodeName="", softTaskIcon=&T_UI_NPC_question, generalConfig {defaultMessage="Sure, what can I do ye for?"}}
  sub_nodes [ConversationChoiceNode_2 {class=ConversationChoiceNode, properties {defaultChoiceDisplayText=Ask about...}}]
 }
 ConversationLinkNode_0 {
  id=#6
  class=ConversationLinkNode
  properties {remoteEntryTag {tagName=Conversation.Entry.RageDebris.Dialogue.BaronEsc}}
 }
 Speech_C_1 {
  id=#5
  class=Speech_C
  outputs [#7]
  properties {
   nodeName=""
   softTaskIcon=&T_UI_NPC_question
   generalConfig {defaultMessage=I do the gold man a favor here and there.}
  }
  sub_nodes [ConversationChoiceNode_1 {class=ConversationChoiceNode, properties {defaultChoiceDisplayText=Midas}}]
 }
 Speech_C_2 {
  id=#7
  class=Speech_C
  outputs [#8]
  properties {
   nodeName=""
   softTaskIcon=&T_UI_NPC_question
   generalConfig {defaultMessage=He pays well and he don't ask questions I don't wanna answer.}
  }
 }
 Speech_C_3 {
  id=#8
  class=Speech_C
  outputs [#9]
  properties {
   nodeName=""
   softTaskIcon=&T_UI_NPC_question
   generalConfig {defaultMessage="Also, he's said some VERY kind things about Mamma PossPoss."}
  }
 }
 Speech_C_4 {
  id=#9
  class=Speech_C
  outputs [#10]
  properties {nodeName="", softTaskIcon=&T_UI_NPC_question, generalConfig {defaultMessage=How can I say naw to that?}}
 }
 ReturnToStart_C_23 {
  id=#10
  class=ReturnToStart_C
  properties {softTaskIcon=&T_UI_NPC_Back}
  sub_nodes [ConversationChoiceNode_32 {class=ConversationChoiceNode, properties {defaultChoiceDisplayText=BACK}}]
 }
 Back_C_0 {
  id=#4
  class=Back_C
  sub_nodes [ConversationChoiceNode_7 {class=ConversationChoiceNode, properties {defaultChoiceDisplayText=BACK}}]
 }
 Speech_C_7 {
  id=#3
  class=Speech_C
  outputs [#11]
  properties {
   nodeName=""
   softTaskIcon=&T_UI_NPC_question
   generalConfig {defaultMessage=Why do the young'uns and old dogs call me Skillet? Funny story.}
  }
  sub_nodes [ConversationChoiceNode_0 {class=ConversationChoiceNode, properties {defaultChoiceDisplayText=His Name}}]
 }
 Speech_C_15 {
  id=#11
  class=Speech_C
  outputs [#12]
  properties {
   nodeName=""
   softTaskIcon=&T_UI_NPC_question
   generalConfig {defaultMessage="My mamma once said, 'Might as well call that boy Skillet, cuz he's always on fire.'"}
  }
 }
 Speech_C_14 {
  id=#12
  class=Speech_C
  outputs [#13]
  properties {
   nodeName=""
   softTaskIcon=&T_UI_NPC_question
   generalConfig {defaultMessage="She weren't wrong! I made my first grenade when I was barely outta the possum pouch."}
  }
 }
 Speech_C_0 {
  id=#13
  class=Speech_C
  outputs [#10]
  properties {
   nodeName=""
   softTaskIcon=&T_UI_NPC_question
   generalConfig {defaultMessage="Now my name's an international brand! Thanks, Mamma."}
  }
 }
]
